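# Sound cue for the SeaMist Mantispheal jump, a bouncing world-object sound that pairs a boing with a creature moan.
# A Mixer layers one of three random boings with one of eight random jump or ambient moans, boosting the moan layer.
# A Modulator holds volume steady and leaves pitch at default variation.
# It uses a dedicated attenuation asset, the ambient props sound class and a prevent-new concurrency override.
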SoundCue /FNGameplayCues/Sounds/Fort_Test_Sounds/World_Objects/Bouncing/SC_SeaMist_Mantispheal_Jump
root=Modulator_1
nodes [
 {id=Mixer_0, children [Random_0,Random_4], properties {inputVolume [1,1.25]}}
 {id=Modulator_1, children [Mixer_0], properties {volumeMax=1, volumeMin=1}}
 {id=Random_0, children [WavePlayer_8, WavePlayer_9, WavePlayer_10], properties {weights [1,1,1]}}
 {
  id=Random_4
  children [WavePlayer_2, WavePlayer_4, WavePlayer_5, WavePlayer_6, WavePlayer_14, WavePlayer_15, WavePlayer_16, WavePlayer_17]
  properties {weights [1,1,1,1,1,1,1,1]}
 }
 {
  id=WavePlayer_10
  properties {soundWaveAssetPtr=&/SC_SeaMist_Mantispheal_Jump_Boing_03}
 }
 {
  id=WavePlayer_14
  properties {soundWaveAssetPtr=&/SC_SeaMist_Mantispheal_Ambient_Moan_06}
 }
 {
  id=WavePlayer_15
  properties {soundWaveAssetPtr=&/SC_SeaMist_Mantispheal_Ambient_Moan_09}
 }
 {
  id=WavePlayer_16
  properties {soundWaveAssetPtr=&/SC_SeaMist_Mantispheal_Ambient_Moan_10}
 }
 {
  id=WavePlayer_17
  properties {soundWaveAssetPtr=&/SC_SeaMist_Mantispheal_Ambient_Moan_12}
 }
 {
  id=WavePlayer_2
  properties {soundWaveAssetPtr=&/SC_SeaMist_Mantispheal_Jump_Moan_01}
 }
 {
  id=WavePlayer_4
  properties {soundWaveAssetPtr=&/SC_SeaMist_Mantispheal_Jump_Moan_02}
 }
 {
  id=WavePlayer_5
  properties {soundWaveAssetPtr=&/SC_SeaMist_Mantispheal_Jump_Moan_03}
 }
 {
  id=WavePlayer_6
  properties {soundWaveAssetPtr=&/SC_SeaMist_Mantispheal_Jump_Moan_04}
 }
 {
  id=WavePlayer_8
  properties {soundWaveAssetPtr=&/SC_SeaMist_Mantispheal_Jump_Boing_01}
 }
 {
  id=WavePlayer_9
  properties {soundWaveAssetPtr=&/SC_SeaMist_Mantispheal_Jump_Boing_02}
 }
]
properties {
 attenuationSettings=&/ATT_SeaMist_Mantispheal_Jump
 bOverrideConcurrency=true
 maxDistance=6250
 soundClassObject=&Fort_Ambient_Props_Athena
 volumeMultiplier=1.35
 attenuationOverrides {falloffDistance=1000, stereoSpread=500, attenuationShapeExtents {x=500}}
 concurrencyOverrides {resolutionRule=PreventNew}
}
