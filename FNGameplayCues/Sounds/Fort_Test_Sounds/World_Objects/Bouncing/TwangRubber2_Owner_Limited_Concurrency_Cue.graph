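# Sound cue for a rubbery tire-bounce impact on world objects, played in the player foley sound class with distance attenuation and a reverb send.
# A quality switch picks either a random choice of three shortened impact waves or, in its last two slots, a single wave.
# Modulators add slight pitch variation.
# Concurrency allows one instance per owner and blocks new plays, so repeated bounces do not stack.
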SoundCue /FNGameplayCues/Sounds/Fort_Test_Sounds/World_Objects/Bouncing/TwangRubber2_Owner_Limited_Concurrency_Cue
root=QualityLevel_0
nodes [
 {id=Modulator_0, children [Random_0], properties {pitchMax=1, volumeMax=1, volumeMin=1}}
 {id=Modulator_1, children [WavePlayer_7], properties {pitchMax=1, volumeMax=1, volumeMin=1}}
 {id=QualityLevel_0, children [Modulator_0,Modulator_0,Modulator_1,Modulator_1]}
 {id=Random_0, children [WavePlayer_2,WavePlayer_5,WavePlayer_6], properties {weights [1,1,1]}}
 {
  id=WavePlayer_2
  properties {soundWaveAssetPtr=&/Tire_Bounce_Impact_01_Shortened}
 }
 {
  id=WavePlayer_5
  properties {soundWaveAssetPtr=&/Tire_Bounce_Impact_02_Shortened}
 }
 {
  id=WavePlayer_6
  properties {soundWaveAssetPtr=&/Tire_Bounce_Impact_03_Shortened}
 }
 {
  id=WavePlayer_7
  properties {soundWaveAssetPtr=&Tire_Bounce_Impact_01_Shortened}
 }
]
properties {
 bOverrideAttenuation=true
 bOverrideConcurrency=true
 maxDistance=2500
 soundClassObject=&Fort_Player_Foley
 volumeMultiplier=1.3
 attenuationOverrides {falloffDistance=2000, attenuationShapeExtents {x=500}}
 concurrencyOverrides {bLimitToOwner=true, maxCount=1, resolutionRule=PreventNew}
 soundSubmixSends [{
   maxSendLevel=1.2
   minSendLevel=0.8
   soundSubmix=&Fortnite_Reverb_Default
   customSendLevelCurve {editorCurveData {keys [<at 0 0.8>,<at 1 1.2>]}}
  }]
}
